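# M_SimpleSlurp is a Default Lit material in the MaterialGym depth elements, set up for skeletal meshes and mesh deformers.
# Its graph holds a single call to the MF_SimpleSlurp material function, so the effect's logic lives there.
# Automatic usage flagging in the editor is turned off, leaving the usage flags fixed.
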Material /Game/Global/Materials/MaterialGym/Materials/MaterialElements/Depth/SimpleSlurp/M_SimpleSlurp
graph {
 properties {bAutomaticallySetUsageInEditor=false, bUsedWithMeshDeformer=true, bUsedWithSkeletalMesh=true, usedShadingModels=Default Lit}
 nodes [MaterialFunctionCall() {materialFunction=&/MF_SimpleSlurp}]
}
